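# Styled dropdown widget for login corrective-action UI, derived from FortComboBox.
# A hit-test-invisible border with label, arrow and outline covers the inherited ComboBoxString; selection updates the label, PreConstruct applies font size and alignment.
# Focus and hover recolor text and dynamic materials, drive HoverAnim and show a confirm glyph on gamepad.
# An open-list flag keeps focus visuals when the dropdown takes focus.
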WidgetBlueprint /FortniteCoreUI/UI/Login/CorrectiveAction/WBP_ComboBox
animations [
 HoverAnim {
  binding_count=1
  length_seconds=0.2
  object=&self:HoverAnim
  sequence {
   bindings [Outline {
     id=#1
     possessed_class=Image
     tracks [{
       type=MovieSceneWidgetMaterialTrack
       sections [{type=MovieSceneParameterSection, channels [Outline Size Ratio {type=MovieSceneScriptingFloatChannel, keys [[0,0.5],[12,1]]}]}]
      }]
    }]
   display_rate {numerator=60}
  }
 }
 OpenAnim {
  binding_count=1
  length_seconds=0.2
  object=&self:OpenAnim
  sequence {
   bindings [Arrow {
     id=#2
     possessed_class=Image
     tracks [{
       type=MovieScene2DTransformTrack
       sections [{type=MovieScene2DTransformSection, end_frame=4, start_frame=0, channels [Angle {type=MovieSceneScriptingFloatChannel, keys [[0,0],[4,-90]]}]}]
      }]
    }]
   tracks [{
     type=MovieSceneEventTrack
     sections [{
       type=MovieSceneEventTriggerSection
       channels [None {type=MovieSceneScriptingEventChannel, const ["<Struct 'MovieSceneEvent' (0x4676eaf91a0) {}>", 1]}]
      }]
    }]
  }
 }
]
graphs [
 EventGraph {nodes [
   n1 = "On Selection Changed (ComboBox)"!() -> (OutputDelegate, then, SelectedItem:String, SelectionType:ESelectInfo Enum) {is=title K2Node_ComponentBoundEvent}
   > Set p_IsListOpened(p_IsListOpened:bool=false) -> (Output_Get:bool)
   > SetText!(execute, self:Self Object Reference, TextValue:String=@n1.SelectedItem) -> (then)
   > PlaySound2D(Sound=/Game/Sounds/Fort_UI_Sounds/FrontEnd/New_FrontEnd/Fortnite_UI_MainMenu_Generic_01_Cue)
   "On Opening (ComboBox)"() {is=title K2Node_ComponentBoundEvent}
   > Set p_IsListOpened(p_IsListOpened=true)
   n2 = OnInputDeviceChanged!() -> (OutputDelegate, then, bNewInputType:ECommonInputType Enum) {is=title K2Node_CustomEvent}
   > SetActionVisibility()
   RefreshHoverAnim() {is=title K2Node_CustomEvent}
   > n3 = Branch(Condition=Get p_IsHover() -> (p_IsHover:bool))
   > n4 = PlaySound2D(Sound=Get HoverSound() -> (HoverSound:Sound Base Object Reference))
   n5 = "Not Equal (Boolean)"(A=Get p_FocusAnimPlayed() -> (p_FocusAnimPlayed:bool), B=OR Boolean(A=Get p_IsFocused() -> (p_IsFocused:bool), B=Get p_IsHover())) -> (ReturnValue=~)
   Branch(execute=[@n4.then, @n3.else], Condition=@n5)
   > n6 = Set p_FocusAnimPlayed(p_FocusAnimPlayed:bool=@n5) -> (Output_Get:bool)
   > n7 = Branch(Condition=@n6)
   n8 = Get HoverAnim() -> (HoverAnim:Widget Animation Object Reference)
   > PlayAnimationForward(InAnimation=@n8)
   PlayAnimationReverse(execute=@n7.else, InAnimation=@n8)
   Event PreConstruct()
   > n9 = Sequence()
   > n10 = GetDynamicMaterial(self=Get Outline() -> (Outline:Image Object Reference))
   > Set OutlineMaterial(OutlineMaterial:Material Instance Dynamic Object Reference=@n10) -> (Output_Get:Material Instance Dynamic Object Reference)
   > n11 = GetDynamicMaterial(self=Get Arrow() -> (Arrow:Image Object Reference))
   > Set ArrowMaterial(ArrowMaterial:Material Instance Dynamic Object Reference=@n11) -> (Output_Get:Material Instance Dynamic Object Reference)
   SetContentSize(execute=@n9.then_1)
   > SetContentAlignment()
   > SetText(TextValue=GetSelectedOption!(self:"Combo Box (String) Object Reference"=Get ComboBox() -> (ComboBox:"Combo Box (String) Object Reference")) -> (ReturnValue:String))
   > Set Focus Visual!(execute, self, IsFocused:bool=false) -> (then) {is=title K2Node_CallFunction}
   Event OnAddedToFocusPath!() -> (OutputDelegate, then, InFocusEvent:Focus Event Structure) {is=title K2Node_Event}
   > n12 = Set p_IsFocused(p_IsFocused:bool=true) -> (Output_Get:bool)
   > n13 = Set Focus Visual(IsFocused=@n12)
   RefreshHoverAnim(execute=[@n13.then, @n17.then]) {is=title K2Node_CallFunction}
   Event Destruct()
   n14 = Get CommonInputSubsystem()
   > Unbind Event from On Input Method Changed(self=@n14, Delegate=@n2)
   Event OnRemovedFromFocusPath!() -> (OutputDelegate, then, InFocusEvent:Focus Event Structure) {is=title K2Node_Event}
   > n15 = Branch(Condition=Get p_IsListOpened() -> (p_IsListOpened:bool))
   > Set p_IsListOpened(p_IsListOpened=false)
   n16 = Set p_IsFocused(execute=@n15.else, p_IsFocused=false)
   > n17 = Set Focus Visual(IsFocused=@n16)
   Event OnMouseEnter()
   > n18 = Set p_IsHover(p_IsHover:bool=true) -> (Output_Get:bool)
   > n19 = Set p_IsFocused(p_IsFocused=@n18)
   > n20 = Set Focus Visual(IsFocused=@n19)
   RefreshHoverAnim(execute=[@n23.then, @n20.then])
   Event OnMouseLeave!() -> (OutputDelegate, then, MouseEvent:"Pointer Event Structure (by ref)") {is=title K2Node_Event}
   > n21 = Set p_IsHover(p_IsHover=false)
   > n22 = Set p_IsFocused(p_IsFocused=@n21)
   > n23 = Set Focus Visual(IsFocused=@n22)
   Event Construct()
   > Bind Event to On Input Method Changed(self=@n14, Delegate=@n2)
  ]}
 Sequencer Events {nodes [
   PlayClickSound() {is=title K2Node_CustomEvent}
   > SpawnSound2D!(
    execute
    Sound:Sound Base Object Reference=Get ClickSound() -> (ClickSound:Sound Base Object Reference)
    VolumeMultiplier:float=1
    PitchMultiplier:float=1
    StartTime:float=0
    ConcurrencySettings:Sound Concurrency Object Reference
    bPersistAcrossLevelTransition:bool=false
    bAutoDestroy:bool=true
   ) -> (then, ReturnValue:Audio Component Object Reference)
  ]}
 Set Focus Visual {nodes [
   n1 = Set Focus Visual!() -> (then, IsFocused:bool=~) {is=title K2Node_FunctionEntry}
   > SetVectorParameterValue(
    self=Get OutlineMaterial() -> (OutlineMaterial:Material Instance Dynamic Object Reference)
    ParameterName=Get OutlineColorParam() -> (OutlineColorParam:Name=None)
    Value=SelectColor(A=Get OutlineEndColor() -> (OutlineEndColor:Linear Color Structure), B=(R=0,G=0.0908,B=0.9911,A=1), bPickA=@n1)
   )
   n2 = SelectColor(A=(R=1,G=1,B=1,A=1), B=(R=0.6038,G=1,B=1,A=1), bPickA=Get IsFocused() -> (IsFocused:bool))
   > SetColorAndOpacity(self=Get LabelText() -> (LabelText:Common Text Object Reference), InColorAndOpacity=Make Slate Color(SpecifiedColor=@n2))
   > SetVectorParameterValue(
    self=Get ArrowMaterial() -> (ArrowMaterial:Material Instance Dynamic Object Reference)
    ParameterName=Get ArrowColorParam() -> (ArrowColorParam:Name=None)
    Value=@n2
   )
   > SetActionVisibility()
  ]}
 SetActionVisibility {nodes [
   SetActionVisibility() {is=title K2Node_FunctionEntry}
   n1 = "Equal (Enum)"(A=GetCurrentInputType(self=Get CommonInputSubsystem()), B=Gamepad)
   n2 = Get p_IsFocused()
   > SetVisibility(
    self=Get InputActionOverlay() -> (InputActionOverlay:Overlay Object Reference)
    InVisibility=Select(Option 0=Collapsed, Option 1=HitTestInvisible, Index=AND Boolean(A=@n2, B=@n1))
   )
  ]}
 SetContentAlignment {nodes [
   SetContentAlignment() {is=title K2Node_FunctionEntry}
   > Set Justification!(
    execute
    Justification:ETextJustify Enum=Select(Option 0:ETextJustify Enum=Left, Option 1:ETextJustify Enum=Center, Index=Get AlignContentCenter() -> (AlignContentCenter:bool)) -> (ReturnValue:ETextJustify Enum)
    self:Text Layout Widget Object Reference=Get LabelText()
   ) -> (then, Output_Get:ETextJustify Enum=Left)
  ]}
 SetContentSize {nodes [
   SetContentSize() {is=title K2Node_FunctionEntry}
   n1 = Break Slate Font Info!(SlateFontInfo:"Slate Font Info Structure (by ref)"=Get Font!(self:Text Object Reference=Get LabelText()) -> (Font:Slate Font Info Structure)) -> (
    FontObject:Object Reference
    FontMaterial:Object Reference
    OutlineSettings:Font Outline Settings Structure
    TypefaceFontName:Name
    Size:float=~
    LetterSpacing:int
    SkewAmount:float=~
    bForceMonospaced:bool=~
    bMaterialIsStencil:bool=~
    MonospacedWidth:float=~
   ) {is=title K2Node_BreakStruct}
   n2 = Get ContentSize() -> (ContentSize:E_UI_InputField_ContentSize Enum=NewEnumerator0)
   n3 = Select!(NewEnumerator0:int=21, NewEnumerator1:int=24, NewEnumerator2:int=36, Index:E_UI_InputField_ContentSize Enum=@n2) -> (ReturnValue:int)
   > SetFont(
    self=Get LabelText()
    InFontInfo=MakeSlateFontInfo!(
     FontObject:Object Reference=Select(
      NewEnumerator0:Object Reference=&BurbankSmall
      NewEnumerator1:Object Reference=&BurbankSmall
      NewEnumerator2:Object Reference=&Burbank
      Index=@n2
     ) -> (ReturnValue:Object Reference)
     FontMaterial:Object Reference=@n1.FontMaterial
     OutlineSettings:Font Outline Settings Structure=@n1.OutlineSettings
     TypefaceFontName:Name=Select(NewEnumerator0:Name=Bold, NewEnumerator1:Name=Bold, NewEnumerator2:Name=BurbankRegular-Black, Index=@n2) -> (ReturnValue:Name)
     Size:float="To Float (Integer)"(InInt=@n3)
     LetterSpacing:int=@n1.LetterSpacing
     SkewAmount:float=@n1.SkewAmount
     bForceMonospaced:bool=false
     bMaterialIsStencil:bool=false
     MonospacedWidth:float=1
    ) -> (ReturnValue:Slate Font Info Structure)
   )
  ]}
 SetText {nodes [
   n1 = SetText!() -> (then, TextValue:String) {is=title K2Node_FunctionEntry}
   > "SetText (Text)"(self=Get LabelText(), InText="To Text (String)"(InString=@n1))
  ]}
]
variables [
 AlignContentCenter {type=boolean, default=true}
 ArrowColorParam {type=string, default=Color}
 ArrowMaterial {type=MaterialInstanceDynamic}
 ClickSound {type=SoundBase, default=&Fortnite_UI_MainMenu_Generic_01_Cue}
 ContentSize {type=E_UI_InputField_ContentSize, default=Large}
 HoverSound {type=SoundBase, default=/Game/Sounds/Fort_UI_Sounds/FrontEnd/New_FrontEnd/Fortnite_UI_MainMenu_Tab_Hover_01_Cue}
 OutlineColorParam {type=string, default=Outline Color}
 OutlineEndColor {type=LinearColor, default=<rgba 0.0331 0.7011 1 1>}
 OutlineMaterial {type=MaterialInstanceDynamic}
 p_FocusAnimPlayed {type=boolean}
 p_IsFocused {type=boolean}
 p_IsHover {type=boolean}
 p_IsListOpened {type=boolean}
]
widgets {
 info {inheritedWidgetCount=1, parentClass=/Script/FortniteUI.FortComboBox, rootWidgetClass=&Overlay, widgetCount=10}
 widgets [
  Overlay_3 {widgetClassPath=&Overlay, widget=Overlay_3}
  ComboBox {
   widgetClassPath=/Script/UMG.ComboBoxString
   bInherited=true
   bIsVariable=true
   parent=Overlay_3
   widget=ComboBox
   properties {
    hasDownArrow=false
    maxListHeight=300
    navigation=&self:WidgetTree.ComboBox.WidgetNavigation_0
    selectedOption=Default
    contentPadding=<ltrb 0 0 0 0>
    defaultOptions [Default, Option 01, Option 02, Option 03, Option 04]
    font {
     bForceMonospaced=false
     bMaterialIsStencil=false
     fontMaterial=null
     fontObject=&BurbankSmall
     letterSpacing=0
     monospacedWidth=1
     size=24
     skewAmount=0
     typefaceFontName=Black
     outlineSettings {
      bApplyOutlineToDropShadows=false
      bMiteredCorners=false
      bSeparateFillAlpha=false
      outlineMaterial=null
      outlineSize=0
      outlineColor=<rgba 0 0 0 1>
     }
    }
    foregroundColor {colorUseRule=UseColor_Specified, specifiedColor=<rgba 1 1 1 1>}
    itemStyle {
     bUseParentRowBrush=false
     activeBrush {
      bIsDynamicallyLoaded=false
      drawAs=Image
      imageType=FullColor
      mirroring=NoMirror
      resourceName=../../../Engine/Content/Slate/Common/Selection.png
      resourceObject=null
      tiling=NoTile
      imageSize=<8 8>
      margin=<ltrb 0 0 0 0>
      outlineSettings {
       bUseBrushTransparency=false
       roundingType=HalfHeightRadius
       width=0
       color {colorUseRule=UseColor_Specified, specifiedColor=<rgba 0 0 0 0>}
       cornerRadii=<0 0 0 0>
      }
      tintColor {colorUseRule=UseColor_Specified, specifiedColor=<rgba 0 0.0908 0.9911 1>}
      uVRegion {bIsValid=false, max=<0 0>, min=<0 0>}
     }
     activeHighlightedBrush {
      bIsDynamicallyLoaded=false
      drawAs=Image
      imageType=FullColor
      mirroring=NoMirror
      resourceName=../../../Engine/Content/Slate/Common/Selection.png
      resourceObject=null
      tiling=NoTile
      imageSize=<8 8>
      margin=<ltrb 0 0 0 0>
      outlineSettings {
       bUseBrushTransparency=false
       roundingType=HalfHeightRadius
       width=0
       color {colorUseRule=UseColor_Specified, specifiedColor=<rgba 0 0 0 0>}
       cornerRadii=<0 0 0 0>
      }
      tintColor {colorUseRule=UseColor_Specified, specifiedColor=<rgba 0 0.1878 0.6795 1>}
      uVRegion {bIsValid=false, max=<0 0>, min=<0 0>}
     }
     activeHoveredBrush {
      bIsDynamicallyLoaded=false
      drawAs=Image
      imageType=FullColor
      mirroring=NoMirror
      resourceName=../../../Engine/Content/Slate/Common/Selection.png
      resourceObject=null
      tiling=NoTile
      imageSize=<8 8>
      margin=<ltrb 0 0 0 0>
      outlineSettings {
       bUseBrushTransparency=false
       roundingType=HalfHeightRadius
       width=0
       color {colorUseRule=UseColor_Specified, specifiedColor=<rgba 0 0 0 0>}
       cornerRadii=<0 0 0 0>
      }
      tintColor {colorUseRule=UseColor_Specified, specifiedColor=<rgba 0 0.0908 0.9911 1>}
      uVRegion {bIsValid=false, max=<0 0>, min=<0 0>}
     }
     dropIndicator_Above {
      bIsDynamicallyLoaded=false
      drawAs=Box
      imageType=FullColor
      mirroring=NoMirror
      resourceName=../../../Engine/Content/Slate/Common/DropZoneIndicator_Above.png
      resourceObject=null
      tiling=NoTile
      imageSize=<0 0>
      margin=<ltrb 0.625 0.625 0 0>
      outlineSettings {
       bUseBrushTransparency=false
       roundingType=HalfHeightRadius
       width=0
       color {colorUseRule=UseColor_Specified, specifiedColor=<rgba 0 0 0 0>}
       cornerRadii=<0 0 0 0>
      }
      tintColor {colorUseRule=UseColor_Specified, specifiedColor=<rgba 0.0091 0.3515 0.9647 1>}
      uVRegion {bIsValid=false, max=<0 0>, min=<0 0>}
     }
     dropIndicator_Below {
      bIsDynamicallyLoaded=false
      drawAs=Box
      imageType=FullColor
      mirroring=NoMirror
      resourceName=../../../Engine/Content/Slate/Common/DropZoneIndicator_Below.png
      resourceObject=null
      tiling=NoTile
      imageSize=<0 0>
      margin=<ltrb 0.625 0 0 0.625>
      outlineSettings {
       bUseBrushTransparency=false
       roundingType=HalfHeightRadius
       width=0
       color {colorUseRule=UseColor_Specified, specifiedColor=<rgba 0 0 0 0>}
       cornerRadii=<0 0 0 0>
      }
      tintColor {colorUseRule=UseColor_Specified, specifiedColor=<rgba 0.0091 0.3515 0.9647 1>}
      uVRegion {bIsValid=false, max=<0 0>, min=<0 0>}
     }
     dropIndicator_Onto {
      bIsDynamicallyLoaded=false
      drawAs=Box
      imageType=FullColor
      mirroring=NoMirror
      resourceName=../../../Engine/Content/Slate/Common/DropZoneIndicator_Onto.png
      resourceObject=null
      tiling=NoTile
      imageSize=<0 0>
      margin=<ltrb 0.25 0.25 0.25 0.25>
      outlineSettings {
       bUseBrushTransparency=false
       roundingType=HalfHeightRadius
       width=0
       color {colorUseRule=UseColor_Specified, specifiedColor=<rgba 0 0 0 0>}
       cornerRadii=<0 0 0 0>
      }
      tintColor {colorUseRule=UseColor_Specified, specifiedColor=<rgba 0.0091 0.3515 0.9647 1>}
      uVRegion {bIsValid=false, max=<0 0>, min=<0 0>}
     }
     evenRowBackgroundBrush {
      bIsDynamicallyLoaded=false
      drawAs=Image
      imageType=NoImage
      mirroring=NoMirror
      resourceName=None
      resourceObject=null
      tiling=NoTile
      imageSize=<0 0>
      margin=<ltrb 0 0 0 0>
      outlineSettings {
       bUseBrushTransparency=false
       roundingType=HalfHeightRadius
       width=0
       color {colorUseRule=UseColor_Specified, specifiedColor=<rgba 0 0 0 0>}
       cornerRadii=<0 0 0 0>
      }
      tintColor {colorUseRule=UseColor_Specified, specifiedColor=<rgba 0.0009 0.006 0.0395 1>}
      uVRegion {bIsValid=false, max=<0 0>, min=<0 0>}
     }
     evenRowBackgroundHoveredBrush {
      bIsDynamicallyLoaded=false
      drawAs=Image
      imageType=FullColor
      mirroring=NoMirror
      resourceName=../../../Engine/Content/Slate/Common/Selection.png
      resourceObject=null
      tiling=NoTile
      imageSize=<8 8>
      margin=<ltrb 0 0 0 0>
      outlineSettings {
       bUseBrushTransparency=false
       roundingType=HalfHeightRadius
       width=0
       color {colorUseRule=UseColor_Specified, specifiedColor=<rgba 0 0 0 0>}
       cornerRadii=<0 0 0 0>
      }
      tintColor {colorUseRule=UseColor_Specified, specifiedColor=<rgba 0.0024 0.0513 0.3864 1>}
      uVRegion {bIsValid=false, max=<0 0>, min=<0 0>}
     }
     inactiveBrush {
      bIsDynamicallyLoaded=false
      drawAs=Image
      imageType=FullColor
      mirroring=NoMirror
      resourceName=../../../Engine/Content/Slate/Common/Selection.png
      resourceObject=null
      tiling=NoTile
      imageSize=<8 8>
      margin=<ltrb 0 0 0 0>
      outlineSettings {
       bUseBrushTransparency=false
       roundingType=HalfHeightRadius
       width=0
       color {colorUseRule=UseColor_Specified, specifiedColor=<rgba 0 0 0 0>}
       cornerRadii=<0 0 0 0>
      }
      tintColor {colorUseRule=UseColor_Specified, specifiedColor=<rgba 0.0024 0.016 0.117 1>}
      uVRegion {bIsValid=false, max=<0 0>, min=<0 0>}
     }
     inactiveHighlightedBrush {
      bIsDynamicallyLoaded=false
      drawAs=Image
      imageType=FullColor
      mirroring=NoMirror
      resourceName=../../../Engine/Content/Slate/Common/Selection.png
      resourceObject=null
      tiling=NoTile
      imageSize=<8 8>
      margin=<ltrb 0 0 0 0>
      outlineSettings {
       bUseBrushTransparency=false
       roundingType=HalfHeightRadius
       width=0
       color {colorUseRule=UseColor_Specified, specifiedColor=<rgba 0 0 0 0>}
       cornerRadii=<0 0 0 0>
      }
      tintColor {colorUseRule=UseColor_Specified, specifiedColor=<rgba 0.0024 0.016 0.117 1>}
      uVRegion {bIsValid=false, max=<0 0>, min=<0 0>}
     }
     inactiveHoveredBrush {
      bIsDynamicallyLoaded=false
      drawAs=Image
      imageType=FullColor
      mirroring=NoMirror
      resourceName=../../../Engine/Content/Slate/Common/Selection.png
      resourceObject=null
      tiling=NoTile
      imageSize=<8 8>
      margin=<ltrb 0 0 0 0>
      outlineSettings {
       bUseBrushTransparency=false
       roundingType=HalfHeightRadius
       width=0
       color {colorUseRule=UseColor_Specified, specifiedColor=<rgba 0 0 0 0>}
       cornerRadii=<0 0 0 0>
      }
      tintColor {colorUseRule=UseColor_Specified, specifiedColor=<rgba 0.0024 0.0513 0.3864 1>}
      uVRegion {bIsValid=false, max=<0 0>, min=<0 0>}
     }
     oddRowBackgroundBrush {
      bIsDynamicallyLoaded=false
      drawAs=Image
      imageType=NoImage
      mirroring=NoMirror
      resourceName=None
      resourceObject=null
      tiling=NoTile
      imageSize=<0 0>
      margin=<ltrb 0 0 0 0>
      outlineSettings {
       bUseBrushTransparency=false
       roundingType=HalfHeightRadius
       width=0
       color {colorUseRule=UseColor_Specified, specifiedColor=<rgba 0 0 0 0>}
       cornerRadii=<0 0 0 0>
      }
      tintColor {colorUseRule=UseColor_Specified, specifiedColor=<rgba 0.0009 0.006 0.0395 1>}
      uVRegion {bIsValid=false, max=<0 0>, min=<0 0>}
     }
     oddRowBackgroundHoveredBrush {
      bIsDynamicallyLoaded=false
      drawAs=Image
      imageType=FullColor
      mirroring=NoMirror
      resourceName=../../../Engine/Content/Slate/Common/Selection.png
      resourceObject=null
      tiling=NoTile
      imageSize=<8 8>
      margin=<ltrb 0 0 0 0>
      outlineSettings {
       bUseBrushTransparency=false
       roundingType=HalfHeightRadius
       width=0
       color {colorUseRule=UseColor_Specified, specifiedColor=<rgba 0 0 0 0>}
       cornerRadii=<0 0 0 0>
      }
      tintColor {colorUseRule=UseColor_Specified, specifiedColor=<rgba 0.0024 0.0513 0.3864 1>}
      uVRegion {bIsValid=false, max=<0 0>, min=<0 0>}
     }
     parentRowBackgroundBrush {
      bIsDynamicallyLoaded=false
      drawAs=Image
      imageType=NoImage
      mirroring=NoMirror
      resourceName=None
      resourceObject=null
      tiling=NoTile
      imageSize=<32 32>
      margin=<ltrb 0 0 0 0>
      outlineSettings {
       bUseBrushTransparency=false
       roundingType=HalfHeightRadius
       width=0
       color {colorUseRule=UseColor_Specified, specifiedColor=<rgba 0 0 0 0>}
       cornerRadii=<0 0 0 0>
      }
      tintColor {colorUseRule=UseColor_Specified, specifiedColor=<rgba 1 1 1 1>}
      uVRegion {bIsValid=false, max=<0 0>, min=<0 0>}
     }
     parentRowBackgroundHoveredBrush {
      bIsDynamicallyLoaded=false
      drawAs=Image
      imageType=NoImage
      mirroring=NoMirror
      resourceName=None
      resourceObject=null
      tiling=NoTile
      imageSize=<32 32>
      margin=<ltrb 0 0 0 0>
      outlineSettings {
       bUseBrushTransparency=false
       roundingType=HalfHeightRadius
       width=0
       color {colorUseRule=UseColor_Specified, specifiedColor=<rgba 0 0 0 0>}
       cornerRadii=<0 0 0 0>
      }
      tintColor {colorUseRule=UseColor_Specified, specifiedColor=<rgba 1 1 1 1>}
      uVRegion {bIsValid=false, max=<0 0>, min=<0 0>}
     }
     selectedTextColor {colorUseRule=UseColor_Specified, specifiedColor=<rgba 1 1 1 1>}
     selectorFocusedBrush {
      bIsDynamicallyLoaded=false
      drawAs=Border
      imageType=FullColor
      mirroring=NoMirror
      resourceName=../../../Engine/Content/Slate/Common/Selector.png
      resourceObject=null
      tiling=Both
      imageSize=<0 0>
      margin=<ltrb 0.25 0.25 0.25 0.25>
      outlineSettings {
       bUseBrushTransparency=false
       roundingType=HalfHeightRadius
       width=0
       color {colorUseRule=UseColor_Specified, specifiedColor=<rgba 0 0 0 0>}
       cornerRadii=<0 0 0 0>
      }
      tintColor {colorUseRule=UseColor_Specified, specifiedColor=<rgba 0.0331 0.7011 1 1>}
      uVRegion {bIsValid=false, max=<0 0>, min=<0 0>}
     }
     textColor {colorUseRule=UseColor_Specified, specifiedColor=<rgba 0.6038 1 1 1>}
    }
    widgetStyle {
     comboButtonStyle {
      downArrowAlign=VAlign_Center
      buttonStyle {
       clickedSlateSound {resourceObject=null}
       disabled {
        bIsDynamicallyLoaded=false
        drawAs=NoDrawType
        imageType=NoImage
        mirroring=NoMirror
        resourceName=None
        resourceObject=null
        tiling=NoTile
        imageSize=<0 0>
        margin=<ltrb 0 0 0 0>
        outlineSettings {
         bUseBrushTransparency=false
         roundingType=HalfHeightRadius
         width=0
         color {colorUseRule=UseColor_Specified, specifiedColor=<rgba 0 0 0 0>}
         cornerRadii=<0 0 0 0>
        }
        tintColor {colorUseRule=UseColor_Specified, specifiedColor=<rgba 1 1 1 1>}
        uVRegion {bIsValid=false, max=<0 0>, min=<0 0>}
       }
       disabledForeground {colorUseRule=UseColor_Specified, specifiedColor=<rgba 0.5271 0.5271 0.5271 1>}
       hovered {
        bIsDynamicallyLoaded=false
        drawAs=NoDrawType
        imageType=NoImage
        mirroring=NoMirror
        resourceName=None
        resourceObject=/FortniteCoreUI/UI/Login/AgeGate/M_UI_AgeGate_Hover
        tiling=NoTile
        imageSize=<32 32>
        margin=<ltrb 0.5 0 0.5 0>
        outlineSettings {
         bUseBrushTransparency=true
         roundingType=FixedRadius
         width=1
         color {colorUseRule=UseColor_Specified, specifiedColor=<rgba 0.7243 0.7243 0.7243 1>}
         cornerRadii=<4 4 4 4>
        }
        tintColor {colorUseRule=UseColor_Specified, specifiedColor=<rgba 0.7243 0.7243 0.7243 1>}
        uVRegion {bIsValid=false, max=<0 0>, min=<0 0>}
       }
       hoveredForeground {colorUseRule=UseColor_Specified, specifiedColor=<rgba 0 1 1 1>}
       hoveredSlateSound {resourceObject=null}
       normal {
        bIsDynamicallyLoaded=false
        drawAs=NoDrawType
        imageType=NoImage
        mirroring=NoMirror
        resourceName=None
        resourceObject=/FortniteCoreUI/UI/Login/AgeGate/M_UI_AgeGate_Base
        tiling=NoTile
        imageSize=<32 32>
        margin=<ltrb 0.5 0 0.5 0>
        outlineSettings {
         bUseBrushTransparency=true
         roundingType=FixedRadius
         width=1
         color {colorUseRule=UseColor_Specified, specifiedColor=<rgba 0.6951 0.6951 0.6951 1>}
         cornerRadii=<4 4 4 4>
        }
        tintColor {colorUseRule=UseColor_Specified, specifiedColor=<rgba 0.4955 0.4955 0.4955 1>}
        uVRegion {bIsValid=false, max=<0 0>, min=<0 0>}
       }
       normalForeground {colorUseRule=UseColor_Specified, specifiedColor=<rgba 0 1 1 1>}
       normalPadding=<ltrb 0 0 0 0>
       pressed {
        bIsDynamicallyLoaded=false
        drawAs=NoDrawType
        imageType=NoImage
        mirroring=NoMirror
        resourceName=None
        resourceObject=/FortniteCoreUI/UI/Login/AgeGate/M_UI_AgeGate_Pressed
        tiling=NoTile
        imageSize=<32 32>
        margin=<ltrb 0.5 0 0.5 0>
        outlineSettings {
         bUseBrushTransparency=true
         roundingType=FixedRadius
         width=1
         color {colorUseRule=UseColor_Specified, specifiedColor=<rgba 0.7243 0.7243 0.7243 1>}
         cornerRadii=<4 4 4 4>
        }
        tintColor {colorUseRule=UseColor_Specified, specifiedColor=<rgba 0.3843 0.3843 0.3843 1>}
        uVRegion {bIsValid=false, max=<0 0>, min=<0 0>}
       }
       pressedForeground {colorUseRule=UseColor_Specified, specifiedColor=<rgba 0 1 1 1>}
       pressedPadding=<ltrb 0 0 0 0>
       pressedSlateSound {resourceObject=null}
      }
      contentPadding=<ltrb 0 0 0 0>
      downArrowImage {
       bIsDynamicallyLoaded=false
       drawAs=NoDrawType
       imageType=FullColor
       mirroring=NoMirror
       resourceName=../../../Engine/Content/Slate/Common/ComboArrow.png
       resourceObject=null
       tiling=NoTile
       imageSize=<16 20>
       margin=<ltrb 0 0 0 0>
       outlineSettings {
        bUseBrushTransparency=false
        roundingType=HalfHeightRadius
        width=0
        color {colorUseRule=UseColor_Specified, specifiedColor=<rgba 0 0 0 0>}
        cornerRadii=<0 0 0 0>
       }
       tintColor {colorUseRule=UseColor_Specified, specifiedColor=<rgba 1 1 1 1>}
       uVRegion {bIsValid=false, max=<0 0>, min=<0 0>}
      }
      downArrowPadding=<ltrb 0 0 0 0>
      menuBorderBrush {
       bIsDynamicallyLoaded=false
       drawAs=Box
       imageType=FullColor
       mirroring=NoMirror
       resourceName=../../../Engine/Content/Slate/Old/Menu_Background.png
       resourceObject=null
       tiling=NoTile
       imageSize=<0 0>
       margin=<ltrb 0.125 0.125 0.125 0.125>
       outlineSettings {
        bUseBrushTransparency=false
        roundingType=HalfHeightRadius
        width=0
        color {colorUseRule=UseColor_Specified, specifiedColor=<rgba 0 0 0 0>}
        cornerRadii=<0 0 0 0>
       }
       tintColor {colorUseRule=UseColor_Specified, specifiedColor=<rgba 1 1 1 1>}
       uVRegion {bIsValid=false, max=<0 0>, min=<0 0>}
      }
      menuBorderPadding=<ltrb 0 0 0 0>
      shadowColorAndOpacity=<rgba 0 0 0 1>
      shadowOffset=<0 0>
     }
     contentPadding=<ltrb 0 0 0 0>
     menuRowPadding=<ltrb 8 8 8 8>
     pressedSlateSound {resourceObject=null}
     selectionChangeSlateSound {resourceObject=null}
    }
   }
   slot {class=OverlaySlot, properties {horizontalAlignment=HAlign_Fill, verticalAlignment=VAlign_Fill}}
  }
  Border_0 {
   widgetClassPath=&Border
   parent=Overlay_3
   widget=Border_0
   properties {visibility=HitTestInvisible, brushColor=<rgba 0.0024 0.016 0.117 1>, padding=<ltrb 0 0 0 0>}
   slot {class=OverlaySlot, properties {horizontalAlignment=HAlign_Fill, verticalAlignment=VAlign_Fill}}
  }
  Overlay_0 {widgetClassPath=&Overlay, parent=Border_0, widget=Overlay_0, slot {class=BorderSlot, properties {padding=<ltrb 0 0 0 0>}}}
  Outline {
   widgetClassPath=&Image
   bIsVariable=true
   parent=Overlay_0
   widget=Outline
   properties {
    visibility=SelfHitTestInvisible
    brush {
     bIsDynamicallyLoaded=false
     drawAs=Box
     imageType=NoImage
     mirroring=NoMirror
     resourceName=None
     resourceObject=/Game/UI/ContentControls/MaterialInstances/MI_UI_ParentalControls_Input_Border
     tiling=NoTile
     imageSize=<32 32>
     margin=<ltrb 0.5 0.5 0.5 0.5>
     outlineSettings {
      bUseBrushTransparency=false
      roundingType=HalfHeightRadius
      width=0
      color {colorUseRule=UseColor_Specified, specifiedColor=<rgba 0 0 0 0>}
      cornerRadii=<0 0 0 0>
     }
     tintColor {colorUseRule=UseColor_Specified, specifiedColor=<rgba 1 1 1 1>}
     uVRegion {bIsValid=false, max=<0 0>, min=<0 0>}
    }
   }
   slot {class=OverlaySlot, properties {horizontalAlignment=HAlign_Fill, verticalAlignment=VAlign_Fill}}
  }
  HorizontalBox_1 {
   widgetClassPath=&HorizontalBox
   parent=Overlay_0
   widget=HorizontalBox_1
   slot {class=OverlaySlot, properties {horizontalAlignment=HAlign_Fill, verticalAlignment=VAlign_Fill, padding=<ltrb 16 16 16 16>}}
  }
  InputActionOverlay {
   widgetClassPath=&Overlay
   bIsVariable=true
   parent=HorizontalBox_1
   widget=InputActionOverlay
   slot {class=HorizontalBoxSlot, properties {horizontalAlignment=HAlign_Left, verticalAlignment=VAlign_Center, padding=<ltrb 0 0 8 0>}}
  }
  CommonActionWidget {
   widgetClassPath=&CommonActionWidget
   bIsVariable=true
   parent=InputActionOverlay
   widget=CommonActionWidget
   properties {visibility=SelfHitTestInvisible, inputActions [{dataTable=&InputActionDataTable, rowName=Confirm}]}
   slot {class=OverlaySlot}
  }
  LabelText {
   widgetClassPath=&CommonTextBlock
   bIsVariable=true
   parent=HorizontalBox_1
   widget=LabelText
   properties {
    autoWrapText=true
    bIsScrollingEnabled=false
    justification=Center
    mobileFontSizeMultiplier=1.3
    text=Label
    font {
     bForceMonospaced=false
     bMaterialIsStencil=false
     fontMaterial=null
     fontObject=&BurbankSmall
     letterSpacing=0
     monospacedWidth=1
     size=36
     skewAmount=0
     typefaceFontName=Bold
     outlineSettings {
      bApplyOutlineToDropShadows=false
      bMiteredCorners=false
      bSeparateFillAlpha=false
      outlineMaterial=null
      outlineSize=0
      outlineColor=<rgba 0 0 0 1>
     }
    }
    shadowOffset=<0 0>
    strikeBrush {
     bIsDynamicallyLoaded=false
     drawAs=NoDrawType
     imageType=NoImage
     mirroring=NoMirror
     resourceName=None
     resourceObject=null
     tiling=NoTile
     imageSize=<32 32>
     margin=<ltrb 0 0 0 0>
     outlineSettings {
      bUseBrushTransparency=false
      roundingType=HalfHeightRadius
      width=0
      color {colorUseRule=UseColor_Specified, specifiedColor=<rgba 0 0 0 0>}
      cornerRadii=<0 0 0 0>
     }
     tintColor {colorUseRule=UseColor_Specified, specifiedColor=<rgba 1 1 1 1>}
     uVRegion {bIsValid=false, max=<0 0>, min=<0 0>}
    }
   }
   slot {class=HorizontalBoxSlot, properties {verticalAlignment=VAlign_Center, size {sizeRule=Fill, value=1}}}
  }
  Arrow {
   widgetClassPath=&Image
   bIsVariable=true
   parent=HorizontalBox_1
   widget=Arrow
   properties {
    visibility=SelfHitTestInvisible
    brush {
     bIsDynamicallyLoaded=false
     drawAs=Image
     imageType=NoImage
     mirroring=NoMirror
     resourceName=None
     resourceObject=/FortniteCoreUI/UI/Login/AgeGate/M_UI_AgeGateDropDownARrow_Inst
     tiling=NoTile
     imageSize=<16 20>
     margin=<ltrb 0 0 0 0>
     outlineSettings {
      bUseBrushTransparency=false
      roundingType=HalfHeightRadius
      width=0
      color {colorUseRule=UseColor_Specified, specifiedColor=<rgba 0 0 0 0>}
      cornerRadii=<0 0 0 0>
     }
     tintColor {colorUseRule=UseColor_Specified, specifiedColor=<rgba 1 1 1 1>}
     uVRegion {bIsValid=false, max=<0 0>, min=<0 0>}
    }
   }
   slot {class=HorizontalBoxSlot, properties {horizontalAlignment=HAlign_Right, verticalAlignment=VAlign_Center, padding=<ltrb 8 0 0 0>}}
  }
 ]
}
