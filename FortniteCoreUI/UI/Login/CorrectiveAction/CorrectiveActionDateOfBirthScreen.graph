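# Login corrective-action screen that asks an existing user to enter a date of birth for age gating.
# Month and day combo boxes, a year input field and Continue and Privacy Policy buttons are inherited widgets from FortCorrectiveActionDateOfBirthScreen.
# The graph handles only presentation: it shows failure reasons in a hidden red error text and toggles processing visibility on the year field.
# Both button clicks fire an element-interacted event.
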WidgetBlueprint /FortniteCoreUI/UI/Login/CorrectiveAction/CorrectiveActionDateOfBirthScreen
graphs [EventGraph {nodes [
   n1 = "On Clicked (Button_Continue)"!() -> (OutputDelegate, then, Button:Common Button Base Object Reference) {is=title K2Node_ComponentBoundEvent}
   > Fire Event Element Interacted(Element=@n1.Button)
   n2 = "On Clicked (Button_PrivacyPolicy)"!() -> (OutputDelegate, then, Button:Common Button Base Object Reference) {is=title K2Node_ComponentBoundEvent}
   > Fire Event Element Interacted(Element=@n2.Button)
   n3 = Event OnShowFailureReason!() -> (OutputDelegate, then, FailureReason:"Text (by ref)") {is=title K2Node_Event}
   n4 = Get Text_AgeEntryChallengeError() -> (Text_AgeEntryChallengeError:Common Text Object Reference)
   > "SetText (Text)"(self=@n4, InText=@n3.FailureReason)
   > SetVisibility(self=@n4, InVisibility=HitTestInvisible)
   n5 = Event OnShowCorrectiveActionProcessing!() -> (OutputDelegate, then, bShowProcessing:bool) {is=title K2Node_Event}
   > n6 = Cast To WBP_InputField!(execute, Object=Get InputField_Year() -> (InputField_Year:Fort Input Field Object Reference)) -> (then, CastFailed, AsWBP Input Field:WBP Input Field Object Reference) {is=title K2Node_DynamicCast}
   > SetProcessingVisibility!(execute, self:WBP Input Field Object Reference=@n6, IsVisible:bool=@n5.bShowProcessing) -> (then)
  ]}]
variables []
widgets {
 info {
  inheritedWidgetCount=5
  parentClass=/Script/FortniteUI.FortCorrectiveActionDateOfBirthScreen
  rootWidgetClass=&VerticalBox
  widgetCount=19
 }
 widgets [
  Content_AgeEntryScreen {widgetClassPath=&VerticalBox, widget=Content_AgeEntryScreen}
  VerticalBox_1 {
   widgetClassPath=&VerticalBox
   parent=Content_AgeEntryScreen
   widget=VerticalBox_1
   slot {class=VerticalBoxSlot, properties {verticalAlignment=VAlign_Center, size {sizeRule=Fill, value=1}}}
  }
  Text_ExistingUserAgeGate {
   widgetClassPath=&CommonTextBlock
   parent=VerticalBox_1
   widget=Text_ExistingUserAgeGate
   properties {
    applyLineHeightToBottomLine=false
    bIsScrollingEnabled=false
    justification=Center
    mobileFontSizeMultiplier=1.4
    text=Please enter your date of birth. This is to help you have a safe and fun experience whatever your age.
    wrapTextAt=944
    font {
     bForceMonospaced=false
     bMaterialIsStencil=false
     fontMaterial=null
     fontObject=&BurbankSmall
     letterSpacing=0
     monospacedWidth=1
     size=18
     skewAmount=0
     typefaceFontName=Bold
     outlineSettings {
      bApplyOutlineToDropShadows=false
      bMiteredCorners=false
      bSeparateFillAlpha=false
      outlineMaterial=null
      outlineSize=0
      outlineColor=<rgba 0 0 0 1>
     }
    }
    shadowOffset=<0 0>
    strikeBrush {
     bIsDynamicallyLoaded=false
     drawAs=NoDrawType
     imageType=NoImage
     mirroring=NoMirror
     resourceName=None
     resourceObject=null
     tiling=NoTile
     imageSize=<32 32>
     margin=<ltrb 0 0 0 0>
     outlineSettings {
      bUseBrushTransparency=false
      roundingType=HalfHeightRadius
      width=0
      color {colorUseRule=UseColor_Specified, specifiedColor=<rgba 0 0 0 0>}
      cornerRadii=<0 0 0 0>
     }
     tintColor {colorUseRule=UseColor_Specified, specifiedColor=<rgba 1 1 1 1>}
     uVRegion {bIsValid=false, max=<0 0>, min=<0 0>}
    }
   }
   slot {class=VerticalBoxSlot, properties {verticalAlignment=VAlign_Top, padding=<ltrb 0 0 0 32>}}
  }
  HorizontalBox_69 {
   widgetClassPath=&HorizontalBox
   parent=VerticalBox_1
   widget=HorizontalBox_69
   slot {class=VerticalBoxSlot, properties {horizontalAlignment=HAlign_Center, verticalAlignment=VAlign_Top}}
  }
  Overlay_0 {
   widgetClassPath=&Overlay
   parent=HorizontalBox_69
   widget=Overlay_0
   slot {class=HorizontalBoxSlot, properties {horizontalAlignment=HAlign_Center, verticalAlignment=VAlign_Center}}
  }
  Spacer_74 {
   widgetClassPath=&Spacer
   parent=Overlay_0
   widget=Spacer_74
   properties {size=<280 1>}
   slot {class=OverlaySlot, properties {horizontalAlignment=HAlign_Fill, verticalAlignment=VAlign_Fill}}
  }
  ComboBox_Month {
   widgetClassPath=&/WBP_ComboBox
   bInherited=true
   bIsVariable=true
   parent=Overlay_0
   widget=ComboBox_Month
   properties {
    bIsFocusable=true
    navigation=&self:WidgetTree.ComboBox_Month.WidgetNavigation_0
   }
   slot {class=OverlaySlot, properties {horizontalAlignment=HAlign_Fill, verticalAlignment=VAlign_Fill}}
  }
  Overlay_1 {
   widgetClassPath=&Overlay
   parent=HorizontalBox_69
   widget=Overlay_1
   slot {class=HorizontalBoxSlot, properties {horizontalAlignment=HAlign_Center, verticalAlignment=VAlign_Center, padding=<ltrb 24 0 24 0>}}
  }
  Spacer {
   widgetClassPath=&Spacer
   parent=Overlay_1
   widget=Spacer
   properties {size=<280 1>}
   slot {class=OverlaySlot, properties {horizontalAlignment=HAlign_Fill, verticalAlignment=VAlign_Fill}}
  }
  ComboBox_Day {
   widgetClassPath=&WBP_ComboBox
   bInherited=true
   bIsVariable=true
   parent=Overlay_1
   widget=ComboBox_Day
   properties {
    bIsFocusable=true
    navigation=&self:WidgetTree.ComboBox_Day.WidgetNavigation_0
   }
   slot {class=OverlaySlot, properties {horizontalAlignment=HAlign_Fill, verticalAlignment=VAlign_Fill}}
  }
  Overlay_2 {
   widgetClassPath=&Overlay
   parent=HorizontalBox_69
   widget=Overlay_2
   slot {class=HorizontalBoxSlot, properties {horizontalAlignment=HAlign_Center, verticalAlignment=VAlign_Center}}
  }
  Spacer_1 {
   widgetClassPath=&Spacer
   parent=Overlay_2
   widget=Spacer_1
   properties {size=<280 1>}
   slot {class=OverlaySlot, properties {horizontalAlignment=HAlign_Fill, verticalAlignment=VAlign_Fill}}
  }
  InputField_Year {
   widgetClassPath=/Game/UI/ContentControls/WBP_InputField
   bInherited=true
   bIsVariable=true
   parent=Overlay_2
   widget=InputField_Year
   properties {
    alignContentCenter=true
    bIsFocusable=true
    contentSize=Large
    defaultHintTextValue=YYYY
    navigation=&self:WidgetTree.InputField_Year.WidgetNavigation_0
    showActionOnlyOnFocus=true
   }
   slot {class=OverlaySlot, properties {horizontalAlignment=HAlign_Fill, verticalAlignment=VAlign_Fill}}
  }
  VerticalBox_3 {
   widgetClassPath=&VerticalBox
   parent=Content_AgeEntryScreen
   widget=VerticalBox_3
   slot {class=VerticalBoxSlot, properties {verticalAlignment=VAlign_Bottom, padding=<ltrb 0 48 0 0>}}
  }
  Text_AgeEntryChallengeError {
   widgetClassPath=&CommonTextBlock
   bIsVariable=true
   parent=VerticalBox_3
   widget=Text_AgeEntryChallengeError
   properties {
    applyLineHeightToBottomLine=false
    bIsScrollingEnabled=false
    justification=Center
    mobileFontSizeMultiplier=1.4
    text=Failed to send request. Try again later.
    visibility=Hidden
    colorAndOpacity {colorUseRule=UseColor_Specified, specifiedColor=<rgba 1 0.147 0.1329 1>}
    font {
     bForceMonospaced=false
     bMaterialIsStencil=false
     fontMaterial=null
     fontObject=&BurbankSmall
     letterSpacing=0
     monospacedWidth=1
     size=18
     skewAmount=0
     typefaceFontName=Bold
     outlineSettings {
      bApplyOutlineToDropShadows=false
      bMiteredCorners=false
      bSeparateFillAlpha=false
      outlineMaterial=null
      outlineSize=0
      outlineColor=<rgba 0 0 0 1>
     }
    }
    shadowOffset=<0 0>
    strikeBrush {
     bIsDynamicallyLoaded=false
     drawAs=NoDrawType
     imageType=NoImage
     mirroring=NoMirror
     resourceName=None
     resourceObject=null
     tiling=NoTile
     imageSize=<32 32>
     margin=<ltrb 0 0 0 0>
     outlineSettings {
      bUseBrushTransparency=false
      roundingType=HalfHeightRadius
      width=0
      color {colorUseRule=UseColor_Specified, specifiedColor=<rgba 0 0 0 0>}
      cornerRadii=<0 0 0 0>
     }
     tintColor {colorUseRule=UseColor_Specified, specifiedColor=<rgba 1 1 1 1>}
     uVRegion {bIsValid=false, max=<0 0>, min=<0 0>}
    }
   }
   slot {class=VerticalBoxSlot, properties {size {sizeRule=Fill, value=1}}}
  }
  Spacer_281 {widgetClassPath=&Spacer, parent=VerticalBox_3, widget=Spacer_281, properties {size=<1 16>}, slot {class=VerticalBoxSlot}}
  Button_Continue {
   widgetClassPath=/FortUILibrary/UIKit/Buttons/WBP_UIKit_Button_Loud
   bInherited=true
   bIsVariable=true
   parent=VerticalBox_3
   widget=Button_Continue
   properties {
    bIsFocusable=false
    buttonSize=Large
    minWidth=400
    text=Continue
    hoveredSlateSoundOverride {resourceObject=/Game/Sounds/Fort_UI_Sounds/FrontEnd/New_FrontEnd/Fortnite_UI_MainMenu_Tab_Hover_01_Cue}
    pressedSlateSoundOverride {resourceObject=/Game/Sounds/Fort_UI_Sounds/FrontEnd/New_FrontEnd/Fortnite_UI_MainMenu_Generic_01_Cue}
    triggeringInputAction {dataTable=&InputActionDataTable, rowName=GenericFaceButtonLeft}
   }
   slot {class=VerticalBoxSlot, properties {horizontalAlignment=HAlign_Center}}
  }
  Spacer_345 {widgetClassPath=&Spacer, parent=VerticalBox_3, widget=Spacer_345, properties {size=<1 16>}, slot {class=VerticalBoxSlot}}
  Button_PrivacyPolicy {
   widgetClassPath=/FortUILibrary/UIKit/Buttons/WBP_UIKit_Button_Quiet
   bInherited=true
   bIsVariable=true
   parent=VerticalBox_3
   widget=Button_PrivacyPolicy
   properties {text=Privacy Policy, triggeringInputAction {dataTable=&InputActionDataTable, rowName=GenericFaceButtonTop}}
   slot {class=VerticalBoxSlot, properties {horizontalAlignment=HAlign_Center}}
  }
 ]
}
