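# Corrective-action screen telling an underage player they have been assigned a new Display Name, changeable once a parent or guardian completes account setup.
# Extends FortCorrectiveActionDisplayNameScreen, laying out the inherited name text and Continue and Privacy Policy buttons, whose clicks fire Element Interacted.
# Parent events toggle the processing spinner and reveal failure text, which this graph never re-hides.
# The display-name downloading event is left unimplemented.
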WidgetBlueprint /FortniteCoreUI/UI/Login/CorrectiveAction/CorrectiveActionDisplayNameScreen
graphs [EventGraph {nodes [
   n1 = "On Clicked (Button_Continue)"!() -> (OutputDelegate, then, Button:Common Button Base Object Reference) {is=title K2Node_ComponentBoundEvent}
   > Fire Event Element Interacted(Element=@n1.Button)
   n2 = "On Clicked (Button_PrivacyPolicy)"!() -> (OutputDelegate, then, Button:Common Button Base Object Reference) {is=title K2Node_ComponentBoundEvent}
   > Fire Event Element Interacted(Element=@n2.Button)
   n3 = Event OnShowCorrectiveActionProcessing!() -> (OutputDelegate, then, bShowProcessing:bool) {is=title K2Node_Event}
   > SetVisibility(
    self=Get Image_DisplayNameEntrySpinner() -> (Image_DisplayNameEntrySpinner:Image Object Reference)
    InVisibility=Select(Option 0=Hidden, Option 1=HitTestInvisible, Index=@n3.bShowProcessing)
   )
   n4 = Event OnShowFailureReason!() -> (OutputDelegate, then, FailureReason:"Text (by ref)") {is=title K2Node_Event}
   n5 = Get Text_DisplayNameEntryError() -> (Text_DisplayNameEntryError:Common Text Object Reference)
   > "SetText (Text)"(self=@n5, InText=@n4.FailureReason)
   > SetVisibility(self=@n5, InVisibility=HitTestInvisible)
   Event OnShowDisplayNameDownloading!() -> (OutputDelegate, then, bShowDownloading:bool) {is=title K2Node_Event}
  ]}]
variables []
widgets {
 info {
  inheritedWidgetCount=3
  parentClass=/Script/FortniteUI.FortCorrectiveActionDisplayNameScreen
  rootWidgetClass=&VerticalBox
  widgetCount=13
 }
 widgets [
  Content_DisplayNameEntryScreen {widgetClassPath=&VerticalBox, widget=Content_DisplayNameEntryScreen}
  VerticalBox_2 {
   widgetClassPath=&VerticalBox
   parent=Content_DisplayNameEntryScreen
   widget=VerticalBox_2
   slot {class=VerticalBoxSlot, properties {verticalAlignment=VAlign_Top}}
  }
  Text_DisplayNameEntryDescription {
   widgetClassPath=&CommonTextBlock
   parent=VerticalBox_2
   widget=Text_DisplayNameEntryDescription
   properties {
    bIsScrollingEnabled=false
    justification=Center
    lineHeightPercentage=1.29
    mobileFontSizeMultiplier=1.2
    text=This is your new Display Name
    font {
     bForceMonospaced=false
     bMaterialIsStencil=false
     fontMaterial=null
     fontObject=&BurbankSmall
     letterSpacing=0
     monospacedWidth=1
     size=18
     skewAmount=0
     typefaceFontName=Bold
     outlineSettings {
      bApplyOutlineToDropShadows=false
      bMiteredCorners=false
      bSeparateFillAlpha=false
      outlineMaterial=null
      outlineSize=0
      outlineColor=<rgba 0 0 0 1>
     }
    }
    shadowOffset=<0 0>
    strikeBrush {
     bIsDynamicallyLoaded=false
     drawAs=NoDrawType
     imageType=NoImage
     mirroring=NoMirror
     resourceName=None
     resourceObject=null
     tiling=NoTile
     imageSize=<32 32>
     margin=<ltrb 0 0 0 0>
     outlineSettings {
      bUseBrushTransparency=false
      roundingType=HalfHeightRadius
      width=0
      color {colorUseRule=UseColor_Specified, specifiedColor=<rgba 0 0 0 0>}
      cornerRadii=<0 0 0 0>
     }
     tintColor {colorUseRule=UseColor_Specified, specifiedColor=<rgba 1 1 1 1>}
     uVRegion {bIsValid=false, max=<0 0>, min=<0 0>}
    }
   }
   slot {class=VerticalBoxSlot, properties {horizontalAlignment=HAlign_Center, verticalAlignment=VAlign_Top, padding=<ltrb 0 0 0 16>}}
  }
  Text_DisplayName {
   widgetClassPath=&CommonTextBlock
   bInherited=true
   parent=VerticalBox_2
   widget=Text_DisplayName
   properties {
    applyLineHeightToBottomLine=false
    bIsScrollingEnabled=false
    justification=Center
    mobileFontSizeMultiplier=1.2
    text=Text Block
    font {
     bForceMonospaced=false
     bMaterialIsStencil=false
     fontMaterial=null
     fontObject=&HeadingNow
     letterSpacing=10
     monospacedWidth=1
     size=25.5
     skewAmount=0
     typefaceFontName=75 Medium
     outlineSettings {
      bApplyOutlineToDropShadows=false
      bMiteredCorners=false
      bSeparateFillAlpha=false
      outlineMaterial=null
      outlineSize=0
      outlineColor=<rgba 0 0 0 1>
     }
    }
    shadowOffset=<0 0>
    strikeBrush {
     bIsDynamicallyLoaded=false
     drawAs=NoDrawType
     imageType=NoImage
     mirroring=NoMirror
     resourceName=None
     resourceObject=null
     tiling=NoTile
     imageSize=<32 32>
     margin=<ltrb 0 0 0 0>
     outlineSettings {
      bUseBrushTransparency=false
      roundingType=HalfHeightRadius
      width=0
      color {colorUseRule=UseColor_Specified, specifiedColor=<rgba 0 0 0 0>}
      cornerRadii=<0 0 0 0>
     }
     tintColor {colorUseRule=UseColor_Specified, specifiedColor=<rgba 1 1 1 1>}
     uVRegion {bIsValid=false, max=<0 0>, min=<0 0>}
    }
   }
   slot {class=VerticalBoxSlot, properties {horizontalAlignment=HAlign_Center}}
  }
  VerticalBox_1 {
   widgetClassPath=&VerticalBox
   parent=Content_DisplayNameEntryScreen
   widget=VerticalBox_1
   slot {class=VerticalBoxSlot, properties {verticalAlignment=VAlign_Center, padding=<ltrb 0 32 0 32>, size {sizeRule=Fill, value=1}}}
  }
  Text_DisplayNameEntryDescription_1 {
   widgetClassPath=&CommonTextBlock
   parent=VerticalBox_1
   widget=Text_DisplayNameEntryDescription_1
   properties {
    applyLineHeightToBottomLine=false
    bIsScrollingEnabled=false
    justification=Center
    lineHeightPercentage=1.29
    mobileFontSizeMultiplier=1.2
    text="Due to your age, we have assigned you a new Display Name. You can change it after your parent or guardian completes your account’s setup."
    wrapTextAt=944
    font {
     bForceMonospaced=false
     bMaterialIsStencil=false
     fontMaterial=null
     fontObject=&BurbankSmall
     letterSpacing=0
     monospacedWidth=1
     size=18
     skewAmount=0
     typefaceFontName=Bold
     outlineSettings {
      bApplyOutlineToDropShadows=false
      bMiteredCorners=false
      bSeparateFillAlpha=false
      outlineMaterial=null
      outlineSize=0
      outlineColor=<rgba 0 0 0 1>
     }
    }
    shadowOffset=<0 0>
    strikeBrush {
     bIsDynamicallyLoaded=false
     drawAs=NoDrawType
     imageType=NoImage
     mirroring=NoMirror
     resourceName=None
     resourceObject=null
     tiling=NoTile
     imageSize=<32 32>
     margin=<ltrb 0 0 0 0>
     outlineSettings {
      bUseBrushTransparency=false
      roundingType=HalfHeightRadius
      width=0
      color {colorUseRule=UseColor_Specified, specifiedColor=<rgba 0 0 0 0>}
      cornerRadii=<0 0 0 0>
     }
     tintColor {colorUseRule=UseColor_Specified, specifiedColor=<rgba 1 1 1 1>}
     uVRegion {bIsValid=false, max=<0 0>, min=<0 0>}
    }
   }
   slot {class=VerticalBoxSlot, properties {verticalAlignment=VAlign_Top, padding=<ltrb 0 0 0 32>}}
  }
  Text_DisplayNameEntryDescription_2 {
   widgetClassPath=&CommonTextBlock
   parent=VerticalBox_1
   widget=Text_DisplayNameEntryDescription_2
   properties {
    applyLineHeightToBottomLine=false
    bIsScrollingEnabled=false
    justification=Center
    mobileFontSizeMultiplier=1.3
    text="If you previously had an Epic Display Name, you can get it back after your parent or guardian completes account setup by logging into your account and selecting the Edit Display Name option."
    wrapTextAt=944
    font {
     bForceMonospaced=false
     bMaterialIsStencil=false
     fontMaterial=null
     fontObject=&NotoSans
     letterSpacing=0
     monospacedWidth=1
     size=15
     skewAmount=0
     typefaceFontName=NotoSans-Regular
     outlineSettings {
      bApplyOutlineToDropShadows=false
      bMiteredCorners=false
      bSeparateFillAlpha=false
      outlineMaterial=null
      outlineSize=0
      outlineColor=<rgba 0 0 0 1>
     }
    }
    shadowOffset=<0 0>
    strikeBrush {
     bIsDynamicallyLoaded=false
     drawAs=NoDrawType
     imageType=NoImage
     mirroring=NoMirror
     resourceName=None
     resourceObject=null
     tiling=NoTile
     imageSize=<32 32>
     margin=<ltrb 0 0 0 0>
     outlineSettings {
      bUseBrushTransparency=false
      roundingType=HalfHeightRadius
      width=0
      color {colorUseRule=UseColor_Specified, specifiedColor=<rgba 0 0 0 0>}
      cornerRadii=<0 0 0 0>
     }
     tintColor {colorUseRule=UseColor_Specified, specifiedColor=<rgba 1 1 1 1>}
     uVRegion {bIsValid=false, max=<0 0>, min=<0 0>}
    }
   }
   slot {class=VerticalBoxSlot, properties {verticalAlignment=VAlign_Top, padding=<ltrb 0 0 0 24>}}
  }
  Text_DisplayNameEntryDescription_3 {
   widgetClassPath=&CommonTextBlock
   parent=VerticalBox_1
   widget=Text_DisplayNameEntryDescription_3
   properties {
    applyLineHeightToBottomLine=false
    bIsScrollingEnabled=false
    justification=Center
    mobileFontSizeMultiplier=1.3
    text="Your Display Name will be used across Epic’s products including Fortnite, Rocket League and Fall Guys. This Display Name will appear when others view your profile."
    wrapTextAt=944
    font {
     bForceMonospaced=false
     bMaterialIsStencil=false
     fontMaterial=null
     fontObject=&NotoSans
     letterSpacing=0
     monospacedWidth=1
     size=15
     skewAmount=0
     typefaceFontName=NotoSans-Regular
     outlineSettings {
      bApplyOutlineToDropShadows=false
      bMiteredCorners=false
      bSeparateFillAlpha=false
      outlineMaterial=null
      outlineSize=0
      outlineColor=<rgba 0 0 0 1>
     }
    }
    shadowOffset=<0 0>
    strikeBrush {
     bIsDynamicallyLoaded=false
     drawAs=NoDrawType
     imageType=NoImage
     mirroring=NoMirror
     resourceName=None
     resourceObject=null
     tiling=NoTile
     imageSize=<32 32>
     margin=<ltrb 0 0 0 0>
     outlineSettings {
      bUseBrushTransparency=false
      roundingType=HalfHeightRadius
      width=0
      color {colorUseRule=UseColor_Specified, specifiedColor=<rgba 0 0 0 0>}
      cornerRadii=<0 0 0 0>
     }
     tintColor {colorUseRule=UseColor_Specified, specifiedColor=<rgba 1 1 1 1>}
     uVRegion {bIsValid=false, max=<0 0>, min=<0 0>}
    }
   }
   slot {class=VerticalBoxSlot, properties {verticalAlignment=VAlign_Top}}
  }
  VerticalBox_5 {widgetClassPath=&VerticalBox, parent=Content_DisplayNameEntryScreen, widget=VerticalBox_5, slot {class=VerticalBoxSlot}}
  Image_DisplayNameEntrySpinner {
   widgetClassPath=&Image
   bIsVariable=true
   parent=VerticalBox_5
   widget=Image_DisplayNameEntrySpinner
   properties {
    visibility=Hidden
    brush {
     bIsDynamicallyLoaded=false
     drawAs=Image
     imageType=NoImage
     mirroring=NoMirror
     resourceName=None
     resourceObject=/FortUILibrary/UIKit/Throbber/MI_UIKit_Monochrome_Throbber
     tiling=NoTile
     imageSize=<64 64>
     margin=<ltrb 0 0 0 0>
     outlineSettings {
      bUseBrushTransparency=false
      roundingType=HalfHeightRadius
      width=0
      color {colorUseRule=UseColor_Specified, specifiedColor=<rgba 0 0 0 0>}
      cornerRadii=<0 0 0 0>
     }
     tintColor {colorUseRule=UseColor_Specified, specifiedColor=<rgba 1 1 1 1>}
     uVRegion {bIsValid=false, max=<0 0>, min=<0 0>}
    }
   }
   slot {class=VerticalBoxSlot, properties {horizontalAlignment=HAlign_Center, verticalAlignment=VAlign_Top, padding=<ltrb 0 0 0 16>}}
  }
  Text_DisplayNameEntryError {
   widgetClassPath=&CommonTextBlock
   bIsVariable=true
   parent=VerticalBox_5
   widget=Text_DisplayNameEntryError
   properties {
    applyLineHeightToBottomLine=false
    bIsScrollingEnabled=false
    justification=Center
    mobileFontSizeMultiplier=1.4
    text=Failed to send request. Try again later.
    visibility=Hidden
    colorAndOpacity {colorUseRule=UseColor_Specified, specifiedColor=<rgba 1 0.147 0.1329 1>}
    font {
     bForceMonospaced=false
     bMaterialIsStencil=false
     fontMaterial=null
     fontObject=&BurbankSmall
     letterSpacing=0
     monospacedWidth=1
     size=18
     skewAmount=0
     typefaceFontName=Bold
     outlineSettings {
      bApplyOutlineToDropShadows=false
      bMiteredCorners=false
      bSeparateFillAlpha=false
      outlineMaterial=null
      outlineSize=0
      outlineColor=<rgba 0 0 0 1>
     }
    }
    shadowOffset=<0 0>
    strikeBrush {
     bIsDynamicallyLoaded=false
     drawAs=NoDrawType
     imageType=NoImage
     mirroring=NoMirror
     resourceName=None
     resourceObject=null
     tiling=NoTile
     imageSize=<32 32>
     margin=<ltrb 0 0 0 0>
     outlineSettings {
      bUseBrushTransparency=false
      roundingType=HalfHeightRadius
      width=0
      color {colorUseRule=UseColor_Specified, specifiedColor=<rgba 0 0 0 0>}
      cornerRadii=<0 0 0 0>
     }
     tintColor {colorUseRule=UseColor_Specified, specifiedColor=<rgba 1 1 1 1>}
     uVRegion {bIsValid=false, max=<0 0>, min=<0 0>}
    }
   }
   slot {class=VerticalBoxSlot, properties {verticalAlignment=VAlign_Top, padding=<ltrb 0 0 0 16>}}
  }
  Button_Continue {
   widgetClassPath=/FortUILibrary/UIKit/Buttons/WBP_UIKit_Button_Loud
   bInherited=true
   bIsVariable=true
   parent=VerticalBox_5
   widget=Button_Continue
   properties {
    buttonSize=Large
    minWidth=400
    text=Continue
    hoveredSlateSoundOverride {resourceObject=/Game/Sounds/Fort_UI_Sounds/FrontEnd/New_FrontEnd/Fortnite_UI_MainMenu_Tab_Hover_01_Cue}
    pressedSlateSoundOverride {resourceObject=/Game/Sounds/Fort_UI_Sounds/FrontEnd/New_FrontEnd/Fortnite_UI_MainMenu_Generic_01_Cue}
    triggeringInputAction {dataTable=&InputActionDataTable, rowName=GenericFaceButtonLeft}
   }
   slot {class=VerticalBoxSlot, properties {horizontalAlignment=HAlign_Center, verticalAlignment=VAlign_Top, padding=<ltrb 0 0 0 16>}}
  }
  Button_PrivacyPolicy {
   widgetClassPath=/FortUILibrary/UIKit/Buttons/WBP_UIKit_Button_Quiet
   bInherited=true
   bIsVariable=true
   parent=VerticalBox_5
   widget=Button_PrivacyPolicy
   properties {text=Privacy Policy, triggeringInputAction {dataTable=&InputActionDataTable, rowName=GenericFaceButtonTop}}
   slot {class=VerticalBoxSlot, properties {horizontalAlignment=HAlign_Center, verticalAlignment=VAlign_Top}}
  }
 ]
}
