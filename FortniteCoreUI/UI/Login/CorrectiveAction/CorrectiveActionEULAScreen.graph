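# Widget Blueprint for the EULA screen in the login corrective action UI, derived from the native FortCorrectiveActionEULAScreen class.
# It uses the shared EulaWidget as its root, filling the inherited EulaWidget slot.
# The inherited Continue and PrivacyPolicy button slots have no widget assigned in this tree.
# The event graph holds only unconnected default events and there are no variables, so the Blueprint adds no logic of its own.
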WidgetBlueprint /FortniteCoreUI/UI/Login/CorrectiveAction/CorrectiveActionEULAScreen
graphs [EventGraph {nodes [
   Event PreConstruct()
   Event Construct()
   Event Tick()
  ]}]
variables []
widgets {
 info {
  inheritedWidgetCount=3
  parentClass=/Script/FortniteUI.FortCorrectiveActionEULAScreen
  rootWidgetClass=/FortniteCoreUI/UI/Login/Eula/EulaWidget.EulaWidget_C
  widgetCount=3
 }
 widgets [
  EulaWidget {widgetClassPath=/FortniteCoreUI/UI/Login/Eula/EulaWidget, bInherited=true, bIsVariable=true, widget=EulaWidget}
  Button_Continue {widgetClassPath=&CommonButtonBase, bInherited=true, widget=null}
  Button_PrivacyPolicy {widgetClassPath=&CommonButtonBase, bInherited=true, widget=null}
 ]
}
