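# Login corrective-action screen that tells underage players some features need guardian permission and collects a parent or guardian email address.
# The layout stacks explanatory text and the email field above a hidden error line, plus Continue and Privacy Policy buttons with gamepad bindings.
# Graph events from the C++ parent toggle the input field's processing indicator and reveal failure reasons in the error text.
# Both button clicks only fire element-interacted events in this graph.
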WidgetBlueprint /FortniteCoreUI/UI/Login/CorrectiveAction/CorrectiveActionGuardianEmailScreen
graphs [EventGraph {nodes [
   n1 = "On Clicked (Button_Continue)"!() -> (OutputDelegate, then, Button:Common Button Base Object Reference) {is=title K2Node_ComponentBoundEvent}
   > Fire Event Element Interacted(Element=@n1.Button)
   n2 = "On Clicked (Button_PrivacyPolicy)"!() -> (OutputDelegate, then, Button:Common Button Base Object Reference) {is=title K2Node_ComponentBoundEvent}
   > Fire Event Element Interacted(Element=@n2.Button)
   n3 = Event OnShowCorrectiveActionProcessing!() -> (OutputDelegate, then, bShowProcessing:bool) {is=title K2Node_Event}
   > n4 = Cast To WBP_InputField!(execute, Object=Get InputField_GuardianEmailEntry() -> (InputField_GuardianEmailEntry:Fort Input Field Object Reference)) -> (then, CastFailed, AsWBP Input Field:WBP Input Field Object Reference) {is=title K2Node_DynamicCast}
   > SetProcessingVisibility!(execute, self:WBP Input Field Object Reference=@n4, IsVisible:bool=@n3.bShowProcessing) -> (then)
   n5 = Event OnShowFailureReason!() -> (OutputDelegate, then, FailureReason:"Text (by ref)") {is=title K2Node_Event}
   n6 = Get Text_EmailEntryChallengeError() -> (Text_EmailEntryChallengeError:Common Text Object Reference)
   > "SetText (Text)"(self=@n6, InText=@n5.FailureReason)
   > SetVisibility(self=@n6, InVisibility=HitTestInvisible)
  ]}]
variables []
widgets {
 info {
  inheritedWidgetCount=3
  parentClass=/Script/FortniteUI.FortCorrectiveActionGuardianEmailScreen
  rootWidgetClass=&VerticalBox
  widgetCount=9
 }
 widgets [
  Content_GuardianEmailEntryScreen {widgetClassPath=&VerticalBox, widget=Content_GuardianEmailEntryScreen}
  VerticalBox {
   widgetClassPath=&VerticalBox
   parent=Content_GuardianEmailEntryScreen
   widget=VerticalBox
   slot {class=VerticalBoxSlot, properties {horizontalAlignment=HAlign_Center, verticalAlignment=VAlign_Center, size {sizeRule=Fill, value=1}}}
  }
  Text_PermissionRequiredInfo {
   widgetClassPath=&CommonTextBlock
   parent=VerticalBox
   widget=Text_PermissionRequiredInfo
   properties {
    applyLineHeightToBottomLine=false
    bIsScrollingEnabled=false
    justification=Center
    lineHeightPercentage=1.29
    mobileFontSizeMultiplier=1.4
    text="Due to your age, some features are unavailable until your parent or guardian gives you permission to use them."
    wrapTextAt=944
    font {
     bForceMonospaced=false
     bMaterialIsStencil=false
     fontMaterial=null
     fontObject=&BurbankSmall
     letterSpacing=0
     monospacedWidth=1
     size=18
     skewAmount=0
     typefaceFontName=Bold
     outlineSettings {
      bApplyOutlineToDropShadows=false
      bMiteredCorners=false
      bSeparateFillAlpha=false
      outlineMaterial=null
      outlineSize=0
      outlineColor=<rgba 0 0 0 1>
     }
    }
    shadowOffset=<0 0>
    strikeBrush {
     bIsDynamicallyLoaded=false
     drawAs=NoDrawType
     imageType=NoImage
     mirroring=NoMirror
     resourceName=None
     resourceObject=null
     tiling=NoTile
     imageSize=<32 32>
     margin=<ltrb 0 0 0 0>
     outlineSettings {
      bUseBrushTransparency=false
      roundingType=HalfHeightRadius
      width=0
      color {colorUseRule=UseColor_Specified, specifiedColor=<rgba 0 0 0 0>}
      cornerRadii=<0 0 0 0>
     }
     tintColor {colorUseRule=UseColor_Specified, specifiedColor=<rgba 1 1 1 1>}
     uVRegion {bIsValid=false, max=<0 0>, min=<0 0>}
    }
   }
   slot {class=VerticalBoxSlot, properties {verticalAlignment=VAlign_Top, padding=<ltrb 0 0 0 32>}}
  }
  InputField_GuardianEmailEntry {
   widgetClassPath=/Game/UI/ContentControls/WBP_InputField
   bInherited=true
   bIsVariable=true
   parent=VerticalBox
   widget=InputField_GuardianEmailEntry
   properties {
    alignContentCenter=true
    bIsFocusable=true
    defaultHintTextValue=Enter Parent or Guardian email address...
    navigation=&self:WidgetTree.InputField_GuardianEmailEntry.WidgetNavigation_0
   }
   slot {class=VerticalBoxSlot, properties {horizontalAlignment=HAlign_Center, verticalAlignment=VAlign_Top}}
  }
  Text_PermitActionInfo {
   widgetClassPath=&CommonTextBlock
   parent=VerticalBox
   widget=Text_PermitActionInfo
   properties {
    applyLineHeightToBottomLine=false
    bIsScrollingEnabled=false
    justification=Center
    mobileFontSizeMultiplier=1.4
    text=We'll send them an email to let them know you're here and how to give you permission.
    wrapTextAt=944
    font {
     bForceMonospaced=false
     bMaterialIsStencil=false
     fontMaterial=null
     fontObject=&NotoSans
     letterSpacing=0
     monospacedWidth=1
     size=15
     skewAmount=0
     typefaceFontName=NotoSans-Regular
     outlineSettings {
      bApplyOutlineToDropShadows=false
      bMiteredCorners=false
      bSeparateFillAlpha=false
      outlineMaterial=null
      outlineSize=0
      outlineColor=<rgba 0 0 0 1>
     }
    }
    shadowOffset=<0 0>
    strikeBrush {
     bIsDynamicallyLoaded=false
     drawAs=NoDrawType
     imageType=NoImage
     mirroring=NoMirror
     resourceName=None
     resourceObject=null
     tiling=NoTile
     imageSize=<32 32>
     margin=<ltrb 0 0 0 0>
     outlineSettings {
      bUseBrushTransparency=false
      roundingType=HalfHeightRadius
      width=0
      color {colorUseRule=UseColor_Specified, specifiedColor=<rgba 0 0 0 0>}
      cornerRadii=<0 0 0 0>
     }
     tintColor {colorUseRule=UseColor_Specified, specifiedColor=<rgba 1 1 1 1>}
     uVRegion {bIsValid=false, max=<0 0>, min=<0 0>}
    }
   }
   slot {class=VerticalBoxSlot, properties {verticalAlignment=VAlign_Top, padding=<ltrb 0 32 0 0>}}
  }
  VerticalBox_4 {
   widgetClassPath=&VerticalBox
   parent=Content_GuardianEmailEntryScreen
   widget=VerticalBox_4
   slot {class=VerticalBoxSlot, properties {verticalAlignment=VAlign_Bottom, padding=<ltrb 0 48 0 0>}}
  }
  Text_EmailEntryChallengeError {
   widgetClassPath=&CommonTextBlock
   bIsVariable=true
   parent=VerticalBox_4
   widget=Text_EmailEntryChallengeError
   properties {
    applyLineHeightToBottomLine=false
    bIsScrollingEnabled=false
    justification=Center
    mobileFontSizeMultiplier=1.4
    text=Failed to send request. Try again later.
    visibility=Hidden
    colorAndOpacity {colorUseRule=UseColor_Specified, specifiedColor=<rgba 1 0.147 0.1329 1>}
    font {
     bForceMonospaced=false
     bMaterialIsStencil=false
     fontMaterial=null
     fontObject=&BurbankSmall
     letterSpacing=0
     monospacedWidth=1
     size=18
     skewAmount=0
     typefaceFontName=Bold
     outlineSettings {
      bApplyOutlineToDropShadows=false
      bMiteredCorners=false
      bSeparateFillAlpha=false
      outlineMaterial=null
      outlineSize=0
      outlineColor=<rgba 0 0 0 1>
     }
    }
    shadowOffset=<0 0>
    strikeBrush {
     bIsDynamicallyLoaded=false
     drawAs=NoDrawType
     imageType=NoImage
     mirroring=NoMirror
     resourceName=None
     resourceObject=null
     tiling=NoTile
     imageSize=<32 32>
     margin=<ltrb 0 0 0 0>
     outlineSettings {
      bUseBrushTransparency=false
      roundingType=HalfHeightRadius
      width=0
      color {colorUseRule=UseColor_Specified, specifiedColor=<rgba 0 0 0 0>}
      cornerRadii=<0 0 0 0>
     }
     tintColor {colorUseRule=UseColor_Specified, specifiedColor=<rgba 1 1 1 1>}
     uVRegion {bIsValid=false, max=<0 0>, min=<0 0>}
    }
   }
   slot {class=VerticalBoxSlot, properties {verticalAlignment=VAlign_Top, padding=<ltrb 0 0 0 16>}}
  }
  Button_Continue {
   widgetClassPath=/FortUILibrary/UIKit/Buttons/WBP_UIKit_Button_Loud
   bInherited=true
   bIsVariable=true
   parent=VerticalBox_4
   widget=Button_Continue
   properties {
    buttonSize=Large
    minWidth=400
    text=Continue
    hoveredSlateSoundOverride {resourceObject=/Game/Sounds/Fort_UI_Sounds/FrontEnd/New_FrontEnd/Fortnite_UI_MainMenu_Tab_Hover_01_Cue}
    pressedSlateSoundOverride {resourceObject=/Game/Sounds/Fort_UI_Sounds/FrontEnd/New_FrontEnd/Fortnite_UI_MainMenu_Generic_01_Cue}
    triggeringInputAction {dataTable=&InputActionDataTable, rowName=GenericFaceButtonLeft}
   }
   slot {class=VerticalBoxSlot, properties {horizontalAlignment=HAlign_Center, verticalAlignment=VAlign_Top, padding=<ltrb 0 0 0 16>}}
  }
  Button_PrivacyPolicy {
   widgetClassPath=/FortUILibrary/UIKit/Buttons/WBP_UIKit_Button_Quiet
   bInherited=true
   bIsVariable=true
   parent=VerticalBox_4
   widget=Button_PrivacyPolicy
   properties {text=Privacy Policy, triggeringInputAction {dataTable=&InputActionDataTable, rowName=GenericFaceButtonTop}}
   slot {class=VerticalBoxSlot, properties {horizontalAlignment=HAlign_Center, verticalAlignment=VAlign_Top}}
  }
 ]
}
